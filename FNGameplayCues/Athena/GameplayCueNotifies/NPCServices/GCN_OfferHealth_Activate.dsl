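; Niagara burst gameplay cue for activation of the NPC offer-health service, linking its effect to a nearby NPC's mesh.
; On burst, it runs a sphere overlap around the target player pawn, filtered by configured object types and class.
; The first result, cast to a Tandem pawn, supplies its skeletal mesh to the first particle component's "ElbowChatSkeletal" parameter.
; If nothing overlaps, it prints a red debug message.
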(fn OnBurst(Niagara)
  (params (MyTarget :type Actor%20Object%20Reference) (Parameters :type Gameplay%20Cue%20Parameters%20Structure%20%28by%20ref%29) (ParticleComponents :type Array%20of%20Niagara%20Particle%20System%20Component%20Object%20References) (AudioComponents :type Array%20of%20Audio%20Component%20Object%20References) (BurstCameraShakeInstance :type Camera%20Shake%20Base%20Object%20Reference) (BurstDecalInstance :type Decal%20Actor%20Object%20Reference))
  (bind _asfort_player_pawn (Utilities|Casting|CastToFortPlayerPawn MyTarget))
  (bind (_outactors _returnvalue) (Collision|SphereOverlapActors _asfort_player_pawn (Transformation|GetActorLocation _asfort_player_pawn) 1e+03 (Variables|Default|GetobjectTypes) (Variables|Default|GettargetClass)))
  (if _returnvalue
    (for _array_element _outactors
      (bind _asbp_player_pawn_tandem (Utilities|Casting|CastToBP_PlayerPawn_Tandem _array_element))
      (Niagara|SetNiagaraSkeletalMeshComponent (Utilities|Array|Get(acopy) ParticleComponents) "ElbowChatSkeletal" (Variables|/Script/Engine.Character|GetMesh _asbp_player_pawn_tandem))
      (return))
    (else
      (Development|PrintString 0 "NoActorOverlapped" true true "(R=1.000000,G=0.021217,B=0.000000,A=1.000000)" 1e+01))))
